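# Physics asset for the JadeTowelGloss backpack cosmetic, defining capsule bodies and joints for its two ponytails and zipper.
# Left and right ponytails are chains of five dyn segments, with capsules widening toward the tip, each hung from a small base body that joins pack_base through a fully free angular constraint.
# Chain joints and the zipper's joint to pack_base set per-joint asymmetric swing limit angles and zero twist limits.
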PhysicsAsset /BRCosmetics/Accessories/FORT_Backpacks/Backpack_F_MED_JadeTowelGloss/Meshes/F_MED_JadeTowelGloss_Pack_Physics
bodies {
 pack_base [{shapeType=Capsule, length=5.7203, radius=10.7797, shapeName=pack_base_capsule, center=<0 0 -3.6164>, rotation=<rot 0 90 0>}]
 ponytail_01_l_base [{shapeType=Capsule, length=0.505, radius=0.505, shapeName=ponytail_01_l_base_capsule, rotation=<rot 0 -90 0>}]
 ponytail_01_l_dyn [{shapeType=Capsule, length=1.5, radius=1.5, shapeName=ponytail_01_l_dyn_capsule, center=<0.85 0 0>, rotation=<rot 0 90 90>}]
 ponytail_01_r_base [{shapeType=Capsule, length=0.505, radius=0.505, shapeName=ponytail_01_r_base_capsule, rotation=<rot 0 -90 0>}]
 ponytail_01_r_dyn [{shapeType=Capsule, length=1.5, radius=1.5, shapeName=ponytail_01_r_dyn_capsule, center=<0.9 0 0>, rotation=<rot 90 0 0>}]
 ponytail_02_l_dyn [{shapeType=Capsule, length=2, radius=1.8, shapeName=ponytail_02_l_dyn_capsule, center=<1 0 0>, rotation=<rot 90 0 0>}]
 ponytail_02_r_dyn [{shapeType=Capsule, length=2, radius=1.9, shapeName=ponytail_02_r_dyn_capsule, center=<1 0 0>, rotation=<rot 86.8348 -179.9988 -179.9991>}]
 ponytail_03_l_dyn [{shapeType=Capsule, length=3, radius=2.3, shapeName=ponytail_03_l_dyn_capsule, center=<2 0 0>, rotation=<rot 90 0 0>}]
 ponytail_03_r_dyn [{shapeType=Capsule, length=3, radius=2.35, shapeName=ponytail_03_r_dyn_capsule, center=<2 0 0>, rotation=<rot 90 0 0>}]
 ponytail_04_l_dyn [{shapeType=Capsule, length=3.5, radius=2.75, shapeName=ponytail_04_l_dyn_capsule, center=<2.5 0 0>, rotation=<rot 90 0 0>}]
 ponytail_04_r_dyn [{shapeType=Capsule, length=3.5, radius=2.7, shapeName=ponytail_04_r_dyn_capsule, center=<2.5 0 0>, rotation=<rot 90 0 0>}]
 ponytail_05_l_dyn [{shapeType=Capsule, length=4, radius=3.1, shapeName=ponytail_05_l_dyn_capsule, center=<3 0 0>, rotation=<rot 90 0 0>}]
 ponytail_05_r_dyn [{shapeType=Capsule, length=4, radius=3.1, shapeName=ponytail_05_r_dyn_capsule, center=<3 0 0>, rotation=<rot 90 0 0>}]
 zipper_dyn [{shapeType=Capsule, length=1, radius=1, shapeName=zipper_dyn_capsule, center=<0.8 0 0>, rotation=<rot 90 0 0>}]
}
constraints [
 {
  bone1Name=ponytail_01_l_base
  bone2Name=pack_base
  swing1LimitDegrees=45
  swing1Motion=Free
  swing2LimitDegrees=45
  swing2Motion=Free
  twistLimitDegrees=45
  twistMotion=Free
 }
 {bone1Name=ponytail_01_l_dyn, bone2Name=ponytail_01_l_base, swing1LimitDegrees=8, swing2LimitDegrees=25, twistLimitDegrees=0}
 {bone1Name=ponytail_02_l_dyn, bone2Name=ponytail_01_l_dyn, swing1LimitDegrees=18, swing2LimitDegrees=30, twistLimitDegrees=0}
 {bone1Name=ponytail_03_l_dyn, bone2Name=ponytail_02_l_dyn, swing1LimitDegrees=18, swing2LimitDegrees=30, twistLimitDegrees=0}
 {bone1Name=ponytail_04_l_dyn, bone2Name=ponytail_03_l_dyn, swing1LimitDegrees=30, swing2LimitDegrees=20, twistLimitDegrees=0}
 {bone1Name=ponytail_05_l_dyn, bone2Name=ponytail_04_l_dyn, swing1LimitDegrees=25, swing2LimitDegrees=40, twistLimitDegrees=0}
 {
  bone1Name=ponytail_01_r_base
  bone2Name=pack_base
  swing1LimitDegrees=45
  swing1Motion=Free
  swing2LimitDegrees=45
  swing2Motion=Free
  twistLimitDegrees=45
  twistMotion=Free
 }
 {bone1Name=ponytail_01_r_dyn, bone2Name=ponytail_01_r_base, swing1LimitDegrees=7, swing2LimitDegrees=45, twistLimitDegrees=0}
 {bone1Name=ponytail_02_r_dyn, bone2Name=ponytail_01_r_dyn, swing1LimitDegrees=18, swing2LimitDegrees=30, twistLimitDegrees=0}
 {bone1Name=ponytail_03_r_dyn, bone2Name=ponytail_02_r_dyn, swing1LimitDegrees=18, swing2LimitDegrees=30, twistLimitDegrees=0}
 {bone1Name=ponytail_04_r_dyn, bone2Name=ponytail_03_r_dyn, swing1LimitDegrees=30, swing2LimitDegrees=20, twistLimitDegrees=0}
 {bone1Name=ponytail_05_r_dyn, bone2Name=ponytail_04_r_dyn, swing1LimitDegrees=25, swing2LimitDegrees=40, twistLimitDegrees=0}
 {bone1Name=zipper_dyn, bone2Name=pack_base, swing1LimitDegrees=5, swing2LimitDegrees=45, twistLimitDegrees=0}
]
